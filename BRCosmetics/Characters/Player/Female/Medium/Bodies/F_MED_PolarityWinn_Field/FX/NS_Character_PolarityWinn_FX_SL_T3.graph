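# Niagara system for audio-reactive cosmetic FX on the PolarityWinn character, attached to skeletal mesh sockets on arms, hands and legs.
# System audio reactivity, with a fake-loop module available, gates soundwave sprite bursts and modulates hand ring and cube size, brightness and material parameters.
# User.Variant blends color palettes, tinted by time-of-day multiplier.
# User flags adjust for ADS, front-end preview and hidden hands or feet; the debug emitter is disabled.
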NiagaraSystem /BRCosmetics/Characters/Player/Female/Medium/Bodies/F_MED_PolarityWinn_Field/FX/NS_Character_PolarityWinn_FX_SL_T3
dependencies {
 usedDataInterfaces [
  &NiagaraDataInterfaceArrayFloat
  &NiagaraDataInterfaceArrayInt32
  &NiagaraDataInterfaceCurve
  &NiagaraDataInterfaceArrayFloat3
  &NiagaraDataInterfaceArrayColor
  &NiagaraDataInterfacePlatformSet
  &NiagaraDataInterfaceSkeletalMesh
  &NiagaraDataInterfaceColorCurve
 ]
 usedDynamicInputs [
  &ColorFromCurve
  /Niagara/DynamicInputs/Lerp/SmoothLerpOverTime_Float
  &RemapRange
 ]
 usedModules [
  &self:SystemSpawnScript.SystemScriptSource.SystemScriptGraph.NiagaraNodeAssignment_3.SetVariables_#1
  &self:SystemSpawnScript.SystemScriptSource.SystemScriptGraph.NiagaraNodeAssignment_4.SetVariables_#2
  &self:SystemSpawnScript.SystemScriptSource.SystemScriptGraph.NiagaraNodeAssignment_2.SetVariables_#3
  &SystemState
  /Game/Effects/Niagara/Modules/Audio/FN_FakeAudioReactivityLoop
  &self:SystemSpawnScript.SystemScriptSource.SystemScriptGraph.NiagaraNodeAssignment_1.SetVariables_#4
  &self:SystemSpawnScript.SystemScriptSource.SystemScriptGraph.NiagaraNodeAssignment_0.SetVariables_#5
  &self:SystemSpawnScript.SystemScriptSource.SystemScriptGraph.NiagaraNodeAssignment_6.SetVariables_#6
 ]
}
emitters {
 Arm_SoundWaves {
  emitterUpdateScript [
   &EmitterState(Inactive Response=::NewEnumerator1, Life Cycle Mode=::NewEnumerator1, Loop Duration=0.5, UseLoopDelay=false)
   &SpawnBurst_Instantaneous(
    Spawn Count=4
    Spawn Probability=&MakeCustomFloatFromBool(Bool=&SetBoolByFloatComparison(A=$System.AudioReactivity, B=0.1), True Float=1, False Float=0)
    Use Spawn Probability=true
    UseLoopCountLimit=false
   )
   &SpawnBurst_Instantaneous(
    Spawn Count=4
    Spawn Probability=&MakeCustomFloatFromBool(Bool=&SetBoolByFloatComparison(A=$System.AudioReactivity, B=0.1), True Float=1, False Float=0)
    Spawn Time=0.2
    Use Spawn Probability=true
    UseLoopCountLimit=false
   ) {module=SpawnBurst_Instantaneous001}
  ]
  particleSpawnScript [
   &PartitionParticles(Number Of Partitions=4, Particles Per Partition=1, Partition Mode:ENiagara_PartitionMode_v2=::NewEnumerator1)
   &InitializeParticle(
    Color=&LerpLinearColors(
     StartColor=&MergeHelper_ColorSwitch(
      0=<rgba 0.1722 1 0.157 20>
      1=<rgba 0.1734 1 0.3776 20>
      2=<rgba 0.2333 0.7691 0.1123 20>
      3=<rgba 0.1994 1 0.371 20>
      Index=$Particles.PartitionParticles.Partition
      Num=::NewEnumerator2
     )
     EndColor=&MergeHelper_ColorSwitch(
      0=<rgba 0.0731 0.3 0.0455 30>
      1=<rgba 0.0162 0.3698 0.1008 20>
      2=<rgba 0.081 0.3 0.0484 30>
      3=<rgba 0.0535 0.6094 0.1171 20>
      Index=$Particles.PartitionParticles.Partition
      Num=::NewEnumerator2
     )
     LerpFactor=$User.Variant
    )
    Lifetime=0.5
    Mesh Scale Mode=::NewEnumerator3
    Mesh Uniform Scale=0.03
    Sprite Size Mode=::NewEnumerator3
    Uniform Sprite Size=35
    UsePositionOffset=false
    Write Material Random=true
   )
   /SetVariables(Particles.VisibilityTag=&MakeCustomIntFromBool(Bool=$User.bIsFrontEndPreview, True INT=1, False INT=0)) {isSetParameters=true, module=SetVariables_#7}
  ]
  particleUpdateScript [
   &ParticleState()
   &SkeletalMeshLocation(
    Bone / Socket Index=&SelectIntFromArray(
     Direct Array Index=$Particles.PartitionParticles.Partition
     Direct Array Mode=::NewEnumerator1
     Int Selection Array=DataInterface() {propertyValues {IntData [13,14,15,16]}}
    )
    MaskSpawnGroup=false
    Mesh Sampling Type=::NewEnumerator5
    Sampled Position Offset=<0 0 0>
    Skeletal Mesh=$System.Skeletal Mesh
    Socket Sampling Mode=::NewEnumerator5
   )
   &CameraOffset(Camera Offset Amount=-12)
   /SetVariables(Particles.OffsetPosition:NiagaraPosition=&AddVectorToPosition(
     Position=$Particles.Position
     Vector=&RotateVectorByQuaternion(Vector To Rotate=<0 0 0>, Quaternion=$Particles.SkeletalMeshLocation.SampledRotation)
    )) {isSetParameters=true, module=SetVariables_#8}
   &DynamicMaterialParameters(
    Index 0 Param 1=&MakeFloatFromVector(Vector=Linked() {linkedVariable {name=Particles.Position, type=NiagaraPosition}}, Channel=::NewEnumerator0)
    Index 0 Param 2=&MakeFloatFromVector(Vector=Linked() {linkedVariable {name=Particles.Position, type=NiagaraPosition}}, Channel=::NewEnumerator1)
    Index 0 Param 3=&MakeFloatFromVector(Vector=Linked() {linkedVariable {name=Particles.Position, type=NiagaraPosition}}, Channel=::NewEnumerator2)
    Index 0 Param 4=&FN_FrontEndFloatModifier(Gameplay=-5, Front End=0)
    Index 1 Param 1=&MakeFloatFromVector(Vector=Linked() {linkedVariable {name=Particles.OffsetPosition, type=NiagaraPosition}}, Channel=::NewEnumerator0)
    Index 1 Param 2=&MakeFloatFromVector(Vector=Linked() {linkedVariable {name=Particles.OffsetPosition, type=NiagaraPosition}}, Channel=::NewEnumerator1)
    Index 1 Param 3=&MakeFloatFromVector(Vector=Linked() {linkedVariable {name=Particles.OffsetPosition, type=NiagaraPosition}}, Channel=::NewEnumerator2)
    Index 1 Param 4=&Lerp_Float(
     A=&FloatFromCurve(
      FloatCurve=DataInterface() {propertyValues {Curve {keys [
          {arriveTangent=-0.0762, interpMode=RCIM_Cubic, leaveTangent=-0.0762, tangentMode=RCTM_User, time=0, value=0.8}
          {arriveTangent=-0.2085, interpMode=RCIM_Cubic, leaveTangent=-0.2085, tangentMode=RCTM_User, time=1, value=0.3}
         ]}}}
      Scale Curve=1.35
     )
     B=&FloatFromCurve(
      FloatCurve=DataInterface() {propertyValues {Curve {keys [
          {arriveTangent=-0.8006, interpMode=RCIM_Cubic, leaveTangent=-0.8006, tangentMode=RCTM_User, time=0, value=0.8}
          {arriveTangent=-0.3233, interpMode=RCIM_Cubic, leaveTangent=-0.3233, tangentMode=RCTM_User, time=1, value=0.4}
         ]}}}
      Scale Curve=1.35
     )
     Alpha=$System.AudioReactivity
    )
    Module.Index0Float0Write=true
    Module.Index0Float1Write=true
    Module.Index0Float2Write=true
    Module.Index0Float3Write=true
    Module.Index1Float0Write=true
    Module.Index1Float1Write=true
    Module.Index1Float2Write=true
    Module.Index1Float3Write=true
    Write Parameter Index 1=::NewEnumerator1
   )
   &ScaleColor(
    Scale Alpha=&FloatFromCurve(FloatCurve=DataInterface() {propertyValues {Curve {keys [
         {arriveTangent=0.5632, interpMode=RCIM_Cubic, leaveTangent=0.5632, tangentMode=RCTM_User, time=0, value=0}
         {arriveTangent=0.168, interpMode=RCIM_Cubic, leaveTangent=0.168, tangentMode=RCTM_User, time=0.25, value=1}
         {arriveTangent=-0.1012, interpMode=RCIM_Cubic, leaveTangent=-0.1012, tangentMode=RCTM_User, time=1, value=0}
        ]}}})
    Scale RGB=&VectorFromFloat(Value=&FloatFromCurve(FloatCurve=DataInterface() {propertyValues {Curve {keys [
          {arriveTangent=0.168, interpMode=RCIM_Cubic, leaveTangent=0.168, tangentMode=RCTM_User, time=0.2, value=1.5}
          {arriveTangent=-0.1012, interpMode=RCIM_Cubic, leaveTangent=-0.1012, tangentMode=RCTM_User, time=0.4, value=1}
         ]}}}))
    ScaleA=true
    ScaleRGB=true
   )
   &ScaleColor(
    Scale Alpha=$System.ReactiveLerp
    Scale RGB=&VectorFromFloat(Value=&Add_Float(
      A=&Multiply_Float(
       A=&FloatFromCurve(
        FloatCurve=DataInterface() {propertyValues {Curve {keys [
            {interpMode=RCIM_Cubic, tangentMode=RCTM_Break, time=0, value=0}
            {arriveTangent=3.7964, interpMode=RCIM_Cubic, leaveTangent=1.1533, tangentMode=RCTM_Break, time=0.9, value=1}
            <at 1 0>
           ]}}}
        CurveIndex=$System.AudioReactivity
       )
       B=&Multiply_Float(
        A=&FN_FrontEndFloatModifier(Gameplay=1, Front End=0.5)
        B=&Lerp_Float(
         A=&MergeHelper_FloatSwitch(0=1, 1=1.5, 2=1, 3=1.5, Index=$Particles.PartitionParticles.Partition, Num=::NewEnumerator2)
         B=&MergeHelper_FloatSwitch(0=2, 1=1, 2=2, 3=1, Index=$Particles.PartitionParticles.Partition, Num=::NewEnumerator2)
         Alpha=$User.Variant
        )
       )
      )
      B=$System.ReactiveLerp
     ))
    ScaleA=true
    ScaleRGB=true
   ) {module=ScaleColor001}
   &ScaleColor(
    Color Value To Scale=$Particles.Color
    Scale Alpha=&MakeFloatFromLinearColor(LinearColor=$System.TODColorMultiplier, Channel:ENiagaraFloatFromLinearColorOptions=::NewEnumerator3)
    Scale RGB=&MakeVectorFromLinearColorRGB(Color=$System.TODColorMultiplier)
    ScaleA=true
    ScaleRGB=true
   ) {module=ScaleColor002}
   &ScaleSpriteSize(
    Scale Sprite Size Mode=::NewEnumerator0
    Uniform Scale Factor=&MakeCustomFloatFromBool(
     Bool=$User.bisHandHidden
     True Float=&SelectFloatFromArray(Direct Array Index=$Particles.PartitionParticles.Partition, Float Selection Array=DataInterface() {propertyValues {FloatData [1,0,1,0]}})
     False Float=&SelectFloatFromArray(Direct Array Index=$Particles.PartitionParticles.Partition, Float Selection Array=DataInterface() {propertyValues {FloatData [1,1,1,1]}})
    )
   )
  ]
  rendererClasses [&NiagaraSpriteRendererProperties]
  renderers [
   {rendererClass=&NiagaraSpriteRendererProperties}
   {rendererClass=&NiagaraSpriteRendererProperties, rendererIndex=1}
   {rendererClass=&NiagaraSpriteRendererProperties, rendererIndex=2}
  ]
 }
 FN_Debug {
  bEnabled=false
  emitterUpdateScript [
   &EmitterState()
   &SpawnBurst_Instantaneous(Spawn Count=1, Use Spawn Probability=false)
  ]
  particleSpawnScript [/Niagara/Modules/Spawn/Initialization/InitializeParticle(
    Lifetime=5
    Module.Write Color=true
    Module.Write Lifetime=true
    Module.Write Mass=true
    Module.Write Material Random=true
    Module.Write Position=false
    Module.Write Scale=false
    Module.Write SpriteRotation=false
    Module.Write SpriteSize=true
    Position=&SimulationPosition()
    Sprite Size=<100 100>
   )]
  particleUpdateScript [
   &ParticleState()
   /SetVariables(Particles.Debug:NiagaraFloat=$System.AudioReactivity) {isSetParameters=true}
   &DynamicMaterialParameters(
    Index 0 Param 1=$Particles.Debug
    Module.Param0WriteEnabled=true
    Module.Param1WriteEnabled=true
    Module.Param2WriteEnabled=true
    Module.Param3WriteEnabled=true
   )
  ]
  rendererClasses [&NiagaraSpriteRendererProperties]
  renderers [{rendererClass=&NiagaraSpriteRendererProperties}]
 }
 Hand_Rings {
  emitterUpdateScript [
   &EmitterState(
    Inactive Response=::NewEnumerator1
    Life Cycle Mode=::NewEnumerator1
    Loop Behavior=::NewEnumerator1
    Loop Duration Mode=::NewEnumerator1
    UseLoopDelay=false
   )
   &SpawnBurst_Instantaneous(Spawn Count=20, Use Spawn Probability=false, UseLoopCountLimit=false)
  ]
  particleSpawnScript [
   &PartitionParticles(Particles Per Partition=5, Partition Mode:ENiagara_PartitionMode_v2=::NewEnumerator1)
   /Niagara/Modules/Spawn/Initialization/V2/InitializeParticle(
    Color=&LerpLinearColors(
     StartColor=&SelectLinearColorFromArray(
      Array Sampling Mode=::NewEnumerator1
      Direct Array Index=$Particles.PartitionParticles.Index In Partition
      Color Selection Array=DataInterface() {propertyValues {ColorData [
         <rgba 0.2639 1.3 0.2415 0.3>
         <rgba 0.1379 1.5 0.2523 0.8>
         <rgba 0.1053 0.4 0.17 0.5623>
         <rgba 0.4107 1.5 0.56 0.3>
         <rgba 0.2231 1.5 0.2486 0.3>
        ]}}
      Random Seed=?
     )
     EndColor=&SelectLinearColorFromArray(
      Array Sampling Mode=::NewEnumerator1
      Direct Array Index=$Particles.PartitionParticles.Index In Partition
      Color Selection Array=DataInterface() {propertyValues {ColorData [
         <rgba 0.0558 0.2 0.0189 2>
         <rgba 0.4793 1.5 0.4621 1>
         <rgba 0.4898 1.5 0.2614 0.8>
         <rgba 0.4856 1.3 0.3242 0.3>
         <rgba 0.553 1.5 0.2978 0.3>
        ]}}
      Random Seed=?
     )
     LerpFactor=$User.Variant
    )
    Lifetime=5
    Mesh Scale Mode=::NewEnumerator3
    Mesh Uniform Scale=&SelectFloatFromArray(
     Direct Array Index=$Particles.PartitionParticles.Index In Partition
     Float Selection Array=DataInterface() {propertyValues {FloatData [0.03,0.018,0.03,0.03,0.018]}}
    )
    Sprite Size Mode=::NewEnumerator3
    UsePositionOffset=false
    Write Material Random=true
   )
   /SetVariables(Particles.VisibilityTag=&SelectIntFromArray(
     Direct Array Index=$Particles.PartitionParticles.Index In Partition
     Int Selection Array=DataInterface() {propertyValues {IntData [0,1,1,2,2]}}
    )) {isSetParameters=true}
   &KillParticles(Kill Particles=&SetBoolByFloatComparison(
     A=&FN_PickFloatDependOnPlatform(
      TargetPlatformActive=-1
      TargetPlatformNoActive=&SelectFloatFromArray(
       Direct Array Index=$Particles.PartitionParticles.Index In Partition
       Float Selection Array=DataInterface() {propertyValues {FloatData [-1,-1,-1,-1,1]}}
      )
     )
     B=0
    ))
  ]
  particleUpdateScript [
   &ParticleState(Kill Particles When Lifetime Has Elapsed=false, Let Infinitely Lived Particles Die When Emitter Deactivates=true)
   &Timeline(
    Play Rate=&FloatFromCurve(FloatCurve=DataInterface() {propertyValues {Curve {keys [<at 0 1>,<at 0.9 2>,<at 1 1>]}}}, CurveIndex=$System.AudioReactivity)
    Timeline End=100000
   )
   &SkeletalMeshLocation(
    Bone / Socket Index=&SelectIntFromArray(Direct Array Index=$Particles.PartitionParticles.Partition, Int Selection Array=DataInterface() {propertyValues {IntData [0,1]}})
    MaskSpawnGroup=false
    Mesh Sampling Type=::NewEnumerator5
    Sampled Position Offset=&MakeCustomVectorFromBool(
     Bool=$User.bIsADS
     True Vector=&SelectVectorFromArray(
      Direct Array Index=$Particles.PartitionParticles.Index In Partition
      Vector Selection Array=DataInterface() {propertyValues {FloatData [<0 0 1.25>, <0 0 0.5>, <0 0 1.5>, <0 0 1.5>, <0 0 0.5>]}}
     )
     False Vector=&SelectVectorFromArray(
      Direct Array Index=$Particles.PartitionParticles.Index In Partition
      Vector Selection Array=DataInterface() {propertyValues {FloatData [<0 0 1.75>, <0 0 1>, <0 0 2>, <0 0 2>, <0 0 1>]}}
     )
    )
    Skeletal Mesh=$System.Skeletal Mesh
    Socket Sampling Mode=::NewEnumerator5
   ) {module=SkeletalMeshLocation001}
   /SetVariables(
    Particles.SpriteAlignment=&RotateVectorByQuaternion(Vector To Rotate=<1 0 0>, Quaternion=$StackContext.SkeletalMeshLocation001.SampledRotation)
    Particles.SpriteFacing=&RotateVectorByQuaternion(Vector To Rotate=<0 0 1>, Quaternion=$StackContext.SkeletalMeshLocation001.SampledRotation)
   ) {isSetParameters=true}
   &CameraOffset(Camera Offset Amount=5)
   &DynamicMaterialParameters(
    Index 0 Param 1=&MakeFloatFromVector(Vector=$Particles.Scale, Channel=::NewEnumerator0)
    Index 0 Param 2=4
    Index 0 Param 3=&Lerp_Float(
     A=0
     B=&Waveform(Global Input Position=-0.17, "Waveform [1]"=::NewEnumerator0, "[1] Phase Scale"=3, "[1] Frequency"=1, "[1] Amplitude Min/Max"=<0 0>)
     Alpha=&Power(a=$System.AudioReactivity, b=3)
    )
    Index 0 Param 4=&Waveform(Global Input Position=128, "Waveform [1]"=::NewEnumerator12, "[1] Phase Scale"=2, "[1] Frequency"=1, "[1] Amplitude Min/Max"=<-0.2 0.2>)
    Index 1 Param 1=3
    Index 1 Param 2=&SelectFloatFromArray(
     Direct Array Index=$Particles.PartitionParticles.Index In Partition
     Float Selection Array=DataInterface() {propertyValues {FloatData [1,1,1,-0.75,0.2]}}
    )
    Index 1 Param 3=&FloatFromCurve(
     FloatCurve=DataInterface() {propertyValues {Curve {keys [{interpMode=RCIM_Cubic, leaveTangent=0.7531, tangentMode=RCTM_Break, time=0, value=0}, <at 1 0.4>]}}}
     CurveIndex=$System.AudioReactivity
    )
    Index 1 Param 4=$Particles.Timeline.Time
    Index 2 Param 1=&FloatFromCurve(
     FloatCurve=DataInterface() {propertyValues {Curve {keys [{interpMode=RCIM_Cubic, tangentMode=RCTM_Break, time=0, value=1}, <at 1 1.2>]}}}
     CurveIndex=$System.AudioReactivity
    )
    Module.Index0Param0Write=true
    Module.Index0Param1Write=true
    Module.Index0Param2Write=true
    Module.Index0Param3Write=true
    Module.Index1Param0Write=true
    Module.Index1Param1Write=true
    Module.Index1Param2Write=true
    Module.Index1Param3Write=true
    Module.Index2Param0Write=true
    Module.Index2Param1Write=true
    Module.Index2Param2Write=true
    Module.Index2Param3Write=true
    Write Parameter Index 1=true
    Write Parameter Index 2=true
   )
   &ScaleMeshSize(Scale Factor=&VectorFromFloat(Value=&Lerp_Float(
      A=&Waveform(Global Input Position=-0.17, "Waveform [1]"=::NewEnumerator0, "[1] Frequency"=1, "[1] Amplitude Min/Max"=<1 1.2>)
      B=&Multiply_Float(
       A=&Waveform(Global Input Position=-0.17, "Waveform [1]"=::NewEnumerator0, "[1] Phase Scale"=5, "[1] Frequency"=1, "[1] Amplitude Min/Max"=<1.8 2.2>)
       B=&SelectFloatFromArray(
        Direct Array Index=$Particles.PartitionParticles.Index In Partition
        Direct Array Mode=::NewEnumerator1
        Float Selection Array=DataInterface() {propertyValues {FloatData [1,0.8,1,0.8,1]}}
       )
      )
      Alpha=&FloatFromCurve(
       FloatCurve=DataInterface() {propertyValues {Curve {keys [{interpMode=RCIM_Cubic, leaveTangent=1.8912, tangentMode=RCTM_Break, time=0, value=0}, <at 1 0.5>]}}}
       CurveIndex=$System.AudioReactivity
      )
     )))
   &ScaleColor(
    Scale RGB=&VectorFromFloat(Value=&Add_Float(
      A=&Multiply_Float(
       A=&FloatFromCurve(
        FloatCurve=DataInterface() {propertyValues {Curve {keys [
            {interpMode=RCIM_Cubic, tangentMode=RCTM_Break, time=0, value=0}
            {arriveTangent=3.7964, interpMode=RCIM_Cubic, leaveTangent=1.1533, tangentMode=RCTM_Break, time=0.9, value=1}
            <at 1 0>
           ]}}}
        CurveIndex=$System.AudioReactivity
        Scale Curve=&SelectFloatFromArray(
         Direct Array Index=$Particles.PartitionParticles.Index In Partition
         Float Selection Array=DataInterface() {propertyValues {FloatData [0,1.6,0.75,2,2]}}
        )
       )
       B=&FN_FrontEndFloatModifier(Gameplay=1, Front End=0.5)
      )
      B=$System.ReactiveLerp
     ))
    ScaleA=true
    ScaleRGB=true
   ) {module=ScaleColor001}
   &ScaleColor(
    Color Value To Scale=$Particles.Color
    Scale Alpha=&MakeFloatFromLinearColor(LinearColor=$System.TODColorMultiplier, Channel:ENiagaraFloatFromLinearColorOptions=::NewEnumerator3)
    Scale RGB=&MakeVectorFromLinearColorRGB(Color=$System.TODColorMultiplier)
    ScaleA=true
    ScaleRGB=true
   ) {module=ScaleColor002}
   &ScaleMeshSize(Scale Factor=&VectorFromFloat(Value=&Lerp_Float(A=1, B=0.7, Alpha=$System.Timeline.NormalizedTime))) {module=ScaleMeshSize002}
   &ScaleMeshSize(Scale Factor=&VectorFromFloat(Value=&MakeCustomFloatFromBool(Bool=$User.bisHandHidden, True Float=0, False Float=1))) {module=ScaleMeshSize001}
   /SetVariables(Particles.DebugDyn:Vector4f=&MakeVector4(X=&Multiply_FloatByInt(Float=1, Integer=$Particles.PartitionParticles.Index In Partition), Y=0, Z=0, W=0)) {enabled=false, isSetParameters=true}
  ]
  rendererClasses [&NiagaraMeshRendererProperties, &NiagaraSpriteRendererProperties]
  renderers [
   {rendererClass=&NiagaraMeshRendererProperties, rendererIndex=0}
   {rendererClass=&NiagaraMeshRendererProperties, rendererIndex=1}
   {rendererClass=&NiagaraMeshRendererProperties, rendererIndex=2}
   {rendererClass=&NiagaraSpriteRendererProperties, rendererIndex=3}
  ]
 }
 Hand_SoundCubes {
  emitterSpawnScript [/SetVariables(Emitter.VariantColor:LinearColor=&LerpLinearColors(StartColor=<rgba 0.1595 2 0.6655 1>, EndColor=<rgba 0.1617 1 0.2257 1>, LerpFactor=$User.Variant)) {isSetParameters=true}]
  emitterUpdateScript [
   &EmitterState()
   &SpawnBurst_Instantaneous(Spawn Count=9, Use Spawn Probability=false, UseLoopCountLimit=false)
   &SpawnBurst_Instantaneous(Spawn Count=9, Spawn Group=1, Use Spawn Probability=false, UseLoopCountLimit=false) {module=SpawnBurst_Instantaneous001}
  ]
  particleSpawnScript [&InitializeParticle(
    Color Mode=::NewEnumerator0
    Lifetime=5
    Mesh Scale Mode=::NewEnumerator3
    Mesh Uniform Scale=0.5
    UsePositionOffset=false
    Write Material Random=true
   )]
  particleUpdateScript [
   &ParticleState()
   &ScaleMeshSize(Scale Factor=&MakeVector(
     X=1
     Y=1
     Z=&Waveform(
      Global Input Position=&RandomRangeFloat(Minimum=0, Maximum=15)
      "Waveform [1]"=::NewEnumerator12
      "[1] Phase Scale"=10
      "[1] Frequency"=1
      "[1] Amplitude Min/Max"=&MakeVector2D(X=0, Y=&Lerp_Float(A=10, B=20, Alpha=$System.AudioReactivity))
     )
    ))
   &SkeletalMeshLocation(
    Bone / Socket Index=&SpawnGroup()
    MaskSpawnGroup=false
    Mesh Sampling Type=::NewEnumerator5
    Sampled Position Offset=<0 0 1.5>
    Skeletal Mesh=$System.Skeletal Mesh
    Socket Sampling Mode=::NewEnumerator5
   )
   /SetVariables(Particles.Position=&AddVectorToPosition(
     Position=$Particles.Position
     Vector=&Multiply_VectorByFloat(
      Vector=&RotateVectorByQuaternion(
       Vector To Rotate=&SelectVectorFromArray(
        Direct Array Index=&ModuloInt(A=&ReturnParticleIDAttributes(), B=9)
        Direct Array Mode=::NewEnumerator1
        Vector Selection Array=DataInterface() {propertyValues {FloatData [<1 1 0>, <1 0 0>, <1 -1 0>, <0 1 0>, <0 0 0>, <0 -1 0>, <-1 1 0>, <-1 0 0>, <-1 -1 0>]}}
       )
       Quaternion=$StackContext.SkeletalMeshLocation.SampledRotation
      )
      Float=0.6091
     )
    )) {isSetParameters=true}
   &ScaleColor(
    Scale Alpha=&Waveform(
     Global Input Position=&RandomRangeFloat(Minimum=0, Maximum=15)
     "Waveform [1]"=::NewEnumerator12
     "[1] Phase Scale"=10
     "[1] Frequency"=1
     "[1] Amplitude Min/Max"=<0 0.5>
    )
    ScaleA=true
    ScaleRGB=true
   )
   &ScaleColor(
    Scale Alpha=&FloatFromCurve(
     FloatCurve=DataInterface() {propertyValues {Curve {keys [
         {interpMode=RCIM_Cubic, leaveTangent=0.431, tangentMode=RCTM_Break, time=0, value=1}
         {arriveTangent=3.4786, interpMode=RCIM_Cubic, leaveTangent=-19.0618, tangentMode=RCTM_Break, time=0.9, value=2}
         <at 1 1>
        ]}}}
     CurveIndex=$System.AudioReactivity
    )
    Scale RGB=&VectorFromFloat(Value=&FloatFromCurve(
      FloatCurve=DataInterface() {propertyValues {Curve {keys [
          {interpMode=RCIM_Cubic, leaveTangent=3.4065, tangentMode=RCTM_Break, time=0, value=1}
          {arriveTangent=0.8198, interpMode=RCIM_Cubic, leaveTangent=-19.0618, tangentMode=RCTM_Break, time=0.9, value=4}
          <at 1 1>
         ]}}}
      CurveIndex=$System.AudioReactivity
     ))
    ScaleA=true
    ScaleRGB=true
   ) {module=ScaleColor001}
   &ScaleColor(
    Color Value To Scale=$Particles.Color
    Scale Alpha=&MakeFloatFromLinearColor(LinearColor=$System.TODColorMultiplier, Channel:ENiagaraFloatFromLinearColorOptions=::NewEnumerator3)
    Scale RGB=&MakeVectorFromLinearColorRGB(Color=$System.TODColorMultiplier)
    ScaleA=true
    ScaleRGB=true
   ) {module=ScaleColor002}
   &ScaleMeshSize(Scale Factor=&VectorFromFloat(Value=&Lerp_Float(A=1, B=0.8, Alpha=$System.Timeline.NormalizedTime))) {module=ScaleMeshSize002}
   &ScaleMeshSize(Scale Factor=&VectorFromFloat(Value=&MakeCustomFloatFromBool(Bool=$User.bisHandHidden, True Float=0, False Float=1))) {module=ScaleMeshSize001}
  ]
  rendererClasses [&NiagaraSpriteRendererProperties, &NiagaraMeshRendererProperties]
  renderers [
   {rendererClass=&NiagaraSpriteRendererProperties}
   {rendererClass=&NiagaraMeshRendererProperties, rendererIndex=1}
  ]
 }
 Leg_SoundWaves {
  emitterUpdateScript [
   &EmitterState(Inactive Response=::NewEnumerator1, Life Cycle Mode=::NewEnumerator1, Loop Duration=0.5, UseLoopDelay=false)
   &SpawnBurst_Instantaneous(
    Spawn Count=6
    Spawn Probability=&MakeCustomFloatFromBool(Bool=&SetBoolByFloatComparison(A=$System.AudioReactivity, B=0.1), True Float=1, False Float=0)
    Use Spawn Probability=true
    UseLoopCountLimit=false
   )
   &SpawnBurst_Instantaneous(
    Spawn Count=6
    Spawn Probability=&MakeCustomFloatFromBool(Bool=&SetBoolByFloatComparison(A=$System.AudioReactivity, B=0.1), True Float=1, False Float=0)
    Spawn Time=0.2
    Use Spawn Probability=true
    UseLoopCountLimit=false
   ) {module=SpawnBurst_Instantaneous001}
  ]
  particleSpawnScript [
   &PartitionParticles(Number Of Partitions=6, Particles Per Partition=1, Partition Mode:ENiagara_PartitionMode_v2=::NewEnumerator1)
   &InitializeParticle(
    Color=&LerpLinearColors(
     StartColor=&MergeHelper_ColorSwitch(
      0=<rgba 0.5043 1 0.1713 20>
      1=<rgba 0.2074 1 0.1616 20>
      2=<rgba 0.0826 1 0.1305 20>
      3=<rgba 0.0615 1 0.2665 20>
      4=<rgba 0.0974 1 0.1693 20>
      5=<rgba 0.3931 1 0.2317 20>
      Index=$Particles.PartitionParticles.Partition
      Num=::NewEnumerator4
     )
     EndColor=&MergeHelper_ColorSwitch(
      0=<rgba 0.0385 0.3 0.0499 30>
      1=<rgba 0.0772 0.5226 0.1087 20>
      2=<rgba 0.0223 0.3005 0.0374 1>
      3=<rgba 0.0426 0.521 0.0938 1>
      4=<rgba 0.0545 0.3005 0.0582 1>
      5=<rgba 0.0235 0.3005 0.0883 1>
      Index=$Particles.PartitionParticles.Partition
      Num=::NewEnumerator4
     )
     LerpFactor=$User.Variant
    )
    Lifetime=0.5
    Sprite Size Mode=::NewEnumerator3
    Uniform Sprite Size=&SelectFloatFromArray(
     Direct Array Index=$Particles.PartitionParticles.Partition
     Float Selection Array=DataInterface() {propertyValues {FloatData [45,40,40,45,40,40]}}
    )
    UsePositionOffset=false
    Write Material Random=true
   )
   /SetVariables(Particles.VisibilityTag=&MakeCustomIntFromBool(Bool=$User.bIsFrontEndPreview, True INT=1, False INT=0)) {isSetParameters=true, module=SetVariables_#7}
  ]
  particleUpdateScript [
   &ParticleState()
   &SkeletalMeshLocation(
    Bone / Socket Index=&SelectIntFromArray(
     Direct Array Index=$Particles.PartitionParticles.Partition
     Direct Array Mode=::NewEnumerator1
     Int Selection Array=DataInterface() {propertyValues {IntData [2,3,4,5,6,7]}}
    )
    MaskSpawnGroup=false
    Mesh Sampling Type=::NewEnumerator5
    Sampled Position Offset=<0 0 0>
    Skeletal Mesh=$System.Skeletal Mesh
    Socket Sampling Mode=::NewEnumerator5
   )
   &CameraOffset(Camera Offset Amount=&FN_FrontEndFloatModifier(Gameplay=20, Front End=0))
   /SetVariables(Particles.OffsetPosition:NiagaraPosition=&AddVectorToPosition(
     Position=$Particles.Position
     Vector=&RotateVectorByQuaternion(Vector To Rotate=<0 0 0>, Quaternion=$Particles.SkeletalMeshLocation.SampledRotation)
    )) {isSetParameters=true, module=SetVariables_#8}
   &DynamicMaterialParameters(
    Index 0 Param 1=&MakeFloatFromVector(Vector=Linked() {linkedVariable {name=Particles.Position, type=NiagaraPosition}}, Channel=::NewEnumerator0)
    Index 0 Param 2=&MakeFloatFromVector(Vector=Linked() {linkedVariable {name=Particles.Position, type=NiagaraPosition}}, Channel=::NewEnumerator1)
    Index 0 Param 3=&MakeFloatFromVector(Vector=Linked() {linkedVariable {name=Particles.Position, type=NiagaraPosition}}, Channel=::NewEnumerator2)
    Index 1 Param 1=&MakeFloatFromVector(Vector=Linked() {linkedVariable {name=Particles.OffsetPosition, type=NiagaraPosition}}, Channel=::NewEnumerator0)
    Index 1 Param 2=&MakeFloatFromVector(Vector=Linked() {linkedVariable {name=Particles.OffsetPosition, type=NiagaraPosition}}, Channel=::NewEnumerator1)
    Index 1 Param 3=&MakeFloatFromVector(Vector=Linked() {linkedVariable {name=Particles.OffsetPosition, type=NiagaraPosition}}, Channel=::NewEnumerator2)
    Index 1 Param 4=&Lerp_Float(
     A=&FloatFromCurve(
      FloatCurve=DataInterface() {propertyValues {Curve {keys [
          {arriveTangent=-0.0762, interpMode=RCIM_Cubic, leaveTangent=-0.0762, tangentMode=RCTM_User, time=0, value=0.8}
          {arriveTangent=-0.2085, interpMode=RCIM_Cubic, leaveTangent=-0.2085, tangentMode=RCTM_User, time=1, value=0.3}
         ]}}}
      Scale Curve=1.35
     )
     B=&FloatFromCurve(
      FloatCurve=DataInterface() {propertyValues {Curve {keys [
          {arriveTangent=-0.8006, interpMode=RCIM_Cubic, leaveTangent=-0.8006, tangentMode=RCTM_User, time=0, value=0.8}
          {arriveTangent=-0.3233, interpMode=RCIM_Cubic, leaveTangent=-0.3233, tangentMode=RCTM_User, time=1, value=0.4}
         ]}}}
      Scale Curve=1.35
     )
     Alpha=$System.AudioReactivity
    )
    Module.Index0Float0Write=true
    Module.Index0Float1Write=true
    Module.Index0Float2Write=true
    Module.Index0Float3Write=true
    Module.Index1Float0Write=true
    Module.Index1Float1Write=true
    Module.Index1Float2Write=true
    Module.Index1Float3Write=true
    Write Parameter Index 1=::NewEnumerator1
   )
   &ScaleColor(
    Scale Alpha=&FloatFromCurve(FloatCurve=DataInterface() {propertyValues {Curve {keys [
         {arriveTangent=0.5632, interpMode=RCIM_Cubic, leaveTangent=0.5632, tangentMode=RCTM_User, time=0, value=0}
         {arriveTangent=0.168, interpMode=RCIM_Cubic, leaveTangent=0.168, tangentMode=RCTM_User, time=0.25, value=1}
         {arriveTangent=-0.1012, interpMode=RCIM_Cubic, leaveTangent=-0.1012, tangentMode=RCTM_User, time=1, value=0}
        ]}}})
    Scale RGB=&VectorFromFloat(Value=&FloatFromCurve(FloatCurve=DataInterface() {propertyValues {Curve {keys [
          {arriveTangent=0.168, interpMode=RCIM_Cubic, leaveTangent=0.168, tangentMode=RCTM_User, time=0.2, value=1.5}
          {arriveTangent=-0.1012, interpMode=RCIM_Cubic, leaveTangent=-0.1012, tangentMode=RCTM_User, time=0.4, value=1}
         ]}}}))
    ScaleA=true
    ScaleRGB=true
   )
   &ScaleColor(
    Scale Alpha=$System.ReactiveLerp
    Scale RGB=&VectorFromFloat(Value=&Add_Float(
      A=&Multiply_Float(
       A=&FloatFromCurve(
        FloatCurve=DataInterface() {propertyValues {Curve {keys [
            {interpMode=RCIM_Cubic, tangentMode=RCTM_Break, time=0, value=0}
            {arriveTangent=3.7964, interpMode=RCIM_Cubic, leaveTangent=1.1533, tangentMode=RCTM_Break, time=0.9, value=1}
            <at 1 0>
           ]}}}
        CurveIndex=$System.AudioReactivity
       )
       B=&FN_FrontEndFloatModifier(Gameplay=0.75, Front End=0.5)
      )
      B=$System.ReactiveLerp
     ))
    ScaleA=true
    ScaleRGB=true
   ) {module=ScaleColor003}
   &ScaleColor(
    Color Value To Scale=$Particles.Color
    Scale Alpha=&MakeFloatFromLinearColor(LinearColor=$System.TODColorMultiplier, Channel:ENiagaraFloatFromLinearColorOptions=::NewEnumerator3)
    Scale RGB=&MakeVectorFromLinearColorRGB(Color=$System.TODColorMultiplier)
    ScaleA=true
    ScaleRGB=true
   ) {module=ScaleColor002}
   &ScaleSpriteSize(
    Scale Sprite Size Mode=::NewEnumerator0
    Uniform Scale Factor=&MakeCustomFloatFromBool(
     Bool=$User.bisFootHidden
     True Float=&SelectFloatFromArray(Direct Array Index=$Particles.PartitionParticles.Partition, Float Selection Array=DataInterface() {propertyValues {FloatData [1,0,0,1,0,0]}})
     False Float=&SelectFloatFromArray(Direct Array Index=$Particles.PartitionParticles.Partition, Float Selection Array=DataInterface() {propertyValues {FloatData [1,1,1,1,1,1]}})
    )
   )
  ]
  rendererClasses [&NiagaraSpriteRendererProperties]
  renderers [
   {rendererClass=&NiagaraSpriteRendererProperties}
   {rendererClass=&NiagaraSpriteRendererProperties, rendererIndex=1}
   {rendererClass=&NiagaraSpriteRendererProperties, rendererIndex=2}
  ]
 }
}
summary {
 systemName=NS_Character_PolarityWinn_FX_SL_T3
 emitters [
  {emitterName=Leg_SoundWaves, rendererClasses [&NiagaraSpriteRendererProperties]}
  {
   emitterName=Hand_SoundCubes
   rendererClasses [&NiagaraSpriteRendererProperties, &NiagaraMeshRendererProperties]
  }
  {emitterName=Hand_Rings, rendererClasses [&NiagaraMeshRendererProperties, &NiagaraSpriteRendererProperties]}
  {emitterName=Arm_SoundWaves, rendererClasses [&NiagaraSpriteRendererProperties]}
  {bEnabled=false, emitterName=FN_Debug, rendererClasses [&NiagaraSpriteRendererProperties]}
 ]
 userVariables [
  User.bIsADS {type=NiagaraBool}
  User.bisFootHidden {type=NiagaraBool}
  User.bIsFrontEnd {type=NiagaraBool}
  User.bIsFrontEndPreview {type=NiagaraBool}
  User.bisHandHidden {type=NiagaraBool}
  User.DebugReactivity {type=NiagaraBool}
  User.EnableFrontendFakeAudioLoop {type=NiagaraFloat, defaultValue=0, description=""}
  User.Variant {type=NiagaraFloat, defaultValue=0, description=""}
 ]
}
