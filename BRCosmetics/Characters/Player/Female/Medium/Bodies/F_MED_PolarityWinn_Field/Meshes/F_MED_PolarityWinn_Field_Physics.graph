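# Physics asset for the PolarityWinn Field medium female player body, covering a small dangling trinket and the right thigh.
# Two sphere bodies form the trinket: a master and an offset, larger child, joined by one constraint with limited swing and almost no twist.
# An unconstrained capsule on the right thigh provides a leg collision shape alongside the trinket.
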PhysicsAsset /BRCosmetics/Characters/Player/Female/Medium/Bodies/F_MED_PolarityWinn_Field/Meshes/F_MED_PolarityWinn_Field_Physics
bodies {
 dyn_trinket_01 [{shapeType=Sphere, radius=2.5, shapeName=dyn_trinket_01_sphere, center=<5.5164 -0.3847 -0.0354>}]
 dyn_trinket_mstr [{shapeType=Sphere, radius=1, shapeName=dyn_trinket_mstr_sphere}]
 thigh_r [{
   shapeType=Capsule
   length=25.8838
   radius=8.0307
   shapeName=thigh_r_capsule
   center=<12.8599 0.8106 -1.063>
   rotation=<rot -85 -0.0013 -9.9996>
  }]
}
constraints [{bone1Name=dyn_trinket_01, bone2Name=dyn_trinket_mstr, swing1LimitDegrees=45, swing2LimitDegrees=35, twistLimitDegrees=1}]
